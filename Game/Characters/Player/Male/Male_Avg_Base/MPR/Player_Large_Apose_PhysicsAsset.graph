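# Physics asset for the large A-pose male player skeleton, giving it a simplified collision and physics body.
# Each body is a single capsule, covering root, pelvis, spine, neck, head and limbs, including twist bones and pinky metacarpals for hands.
# Constraints chain the bodies along the bone hierarchy, with twist bodies branching from their limb bones.
# All joints share identical swing and twist limits rather than per-joint anatomical ranges.
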PhysicsAsset /Game/Characters/Player/Male/Male_Avg_Base/MPR/Player_Large_Apose_PhysicsAsset
bodies {
 ball_l [{shapeType=Capsule, length=6.766, radius=6.2707, center=<5.343 -3.4676 -1.3114>, rotation=<rot -2.0121 -90 9.5104>}]
 ball_r [{shapeType=Capsule, length=7.0469, radius=6.4489, center=<-5.0303 3.3127 1.6251>, rotation=<rot -3.8726 -90 15.0695>}]
 calf_l [{shapeType=Capsule, length=16.4849, radius=5.8484, center=<-15.282 -0.404 -0.4735>, rotation=<rot 0 85.8837 -88.5479>}]
 calf_r [{shapeType=Capsule, length=16.7264, radius=5.8507, center=<15.0331 0.3901 0.4159>, rotation=<rot 0 85.71 -88.8885>}]
 calf_twist_02_l [{shapeType=Capsule, length=8.946, radius=6.2962, center=<-0.2132 -0.9599 -0.2593>, rotation=<rot 0 71.6136 -88.2549>}]
 calf_twist_02_r [{shapeType=Capsule, length=8.946, radius=6.2963, center=<0.213 0.96 0.2593>, rotation=<rot 0 71.6138 -88.2548>}]
 clavicle_l [{shapeType=Capsule, length=12.3053, radius=18.5877, center=<13.4073 -0.6479 0.6314>, rotation=<rot 0 90 -90>}]
 clavicle_r [{shapeType=Capsule, length=12.3045, radius=18.5879, center=<-13.408 0.6482 -0.631>, rotation=<rot 0 -89.9999 -90>}]
 foot_l [{shapeType=Capsule, length=9.4595, radius=9.3128, center=<-7.8255 3.4902 -0.716>, rotation=<rot 0 -137.1331 -89.1628>}]
 foot_r [{shapeType=Capsule, length=9.3952, radius=9.313, center=<7.7817 -3.4435 0.7169>, rotation=<rot 0 42.8668 -90.8371>}]
 hand_l [{shapeType=Capsule, length=5.781, radius=3.8743, center=<1.2989 -1.0514 -0.0681>, rotation=<rot 15.7659 -90 2.9171>}]
 hand_r [{shapeType=Capsule, length=5.7622, radius=3.8588, center=<-1.2909 1.002 0.0862>, rotation=<rot 18.3047 -90 3.0425>}]
 head [{shapeType=Capsule, length=11.6584, radius=10.6161, center=<1.102 1.7606 -0.0192>, rotation=<rot 0 48.3008 -89.6289>}]
 lowerarm_l [{shapeType=Capsule, length=12.2474, radius=10.5923, center=<8.0409 2.054 -1.1112>, rotation=<rot 0 -90.6821 -69.3826>}]
 lowerarm_r [{shapeType=Capsule, length=11.7748, radius=11.5311, center=<-7.1927 -1.3062 0.0139>, rotation=<rot 0 -101.0781 -45.7943>}]
 lowerarm_twist_01_l [{shapeType=Capsule, length=6.6673, radius=5.8556, center=<10.3713 0.4208 0.3252>, rotation=<rot 0 0.3693 -71.3737>}]
 lowerarm_twist_01_r [{shapeType=Capsule, length=6.8764, radius=6.5447, center=<-8.778 -0.6796 -0.2937>, rotation=<rot 0 2.5432 -62.7779>}]
 neck_01 [{shapeType=Capsule, length=16.6584, radius=12.828, center=<-0.3478 1.8549 -0.0273>, rotation=<rot 0.2392 -90 0.074>}]
 neck_02 [{shapeType=Capsule, length=15.15, radius=15.15, rotation=<rot 0 119.9036 -90>}]
 pelvis [{shapeType=Capsule, length=24.0426, radius=19.4946, center=<5.1798 0.0963 0.0006>, rotation=<rot -0.0052 -90 0.002>}]
 pinky_metacarpal_l [{shapeType=Capsule, length=4.012, radius=3.1695, center=<3.5633 -0.0509 -0.4186>, rotation=<rot 0 90 -90>}]
 pinky_metacarpal_r [{shapeType=Capsule, length=4.4686, radius=3.1695, center=<-3.111 0.0511 0.4186>, rotation=<rot 0 -90.001 -89.9998>}]
 root [{shapeType=Capsule, length=15.15, radius=15.15, rotation=<rot -0.772 -90 0>}]
 spine_01 [{shapeType=Capsule, length=15.15, radius=15.15, rotation=<rot 0 97.249 -90>}]
 spine_02 [{shapeType=Capsule, length=4.3653, radius=24.4128, center=<6.8827 7.6782 0>, rotation=<rot 0 91.8162 -90>}]
 spine_03 [{shapeType=Capsule, length=18.8244, radius=25.4487, center=<12.851 -2.7659 -0.0005>, rotation=<rot 0 89.8984 -90>}]
 spine_04 [{shapeType=Capsule, length=6.1155, radius=13.9502, center=<2.2874 16.9329 0.0003>, rotation=<rot 0 90 -90>}]
 spine_05 [{shapeType=Capsule, length=15.8866, radius=10.4419, center=<7.232 0.2125 0.0002>, rotation=<rot 0 25.9505 -89.9986>}]
 thigh_l [{shapeType=Capsule, length=9.5442, radius=7.8748, center=<-27.21 1.131 -0.2195>, rotation=<rot 0 69.0812 -76.8642>}]
 thigh_r [{shapeType=Capsule, length=9.6204, radius=8.0965, center=<27.2346 -1.1175 0.1705>, rotation=<rot 0 64.1307 -79.7863>}]
 thigh_twist_01_l [{shapeType=Capsule, length=13.3439, radius=11.8278, center=<11.5324 0.8241 4.2611>, rotation=<rot 0 178.6416 -76.1327>}]
 thigh_twist_01_r [{shapeType=Capsule, length=13.3454, radius=11.8317, center=<-11.5779 -0.826 -4.26>, rotation=<rot 0 178.5325 -76.3005>}]
 upperarm_l [{shapeType=Capsule, length=15.0515, radius=13.4023, center=<7.5903 -2.7955 -0.9245>, rotation=<rot 0 3.4972 -89.1633>}]
 upperarm_r [{shapeType=Capsule, length=15.0523, radius=13.403, center=<-7.5908 2.7945 0.924>, rotation=<rot 0 3.4982 -89.1664>}]
 upperarm_twist_01_l [{shapeType=Capsule, length=10.9828, radius=8.9494, center=<8.1819 0.5803 -2.3259>, rotation=<rot 0 12.0266 -76.453>}]
 upperarm_twist_01_r [{shapeType=Capsule, length=12.7527, radius=11.7188, center=<-4.2095 2.0658 2.6637>, rotation=<rot 0 7.5645 -85.9543>}]
}
constraints [
 {bone1Name=pelvis, bone2Name=root, swing1LimitDegrees=45, swing2LimitDegrees=45, twistLimitDegrees=45}
 {bone1Name=spine_01, bone2Name=pelvis, swing1LimitDegrees=45, swing2LimitDegrees=45, twistLimitDegrees=45}
 {bone1Name=spine_02, bone2Name=spine_01, swing1LimitDegrees=45, swing2LimitDegrees=45, twistLimitDegrees=45}
 {bone1Name=spine_03, bone2Name=spine_02, swing1LimitDegrees=45, swing2LimitDegrees=45, twistLimitDegrees=45}
 {bone1Name=spine_04, bone2Name=spine_03, swing1LimitDegrees=45, swing2LimitDegrees=45, twistLimitDegrees=45}
 {bone1Name=spine_05, bone2Name=spine_04, swing1LimitDegrees=45, swing2LimitDegrees=45, twistLimitDegrees=45}
 {bone1Name=clavicle_l, bone2Name=spine_05, swing1LimitDegrees=45, swing2LimitDegrees=45, twistLimitDegrees=45}
 {bone1Name=upperarm_l, bone2Name=clavicle_l, swing1LimitDegrees=45, swing2LimitDegrees=45, twistLimitDegrees=45}
 {bone1Name=lowerarm_l, bone2Name=upperarm_l, swing1LimitDegrees=45, swing2LimitDegrees=45, twistLimitDegrees=45}
 {bone1Name=hand_l, bone2Name=lowerarm_l, swing1LimitDegrees=45, swing2LimitDegrees=45, twistLimitDegrees=45}
 {bone1Name=pinky_metacarpal_l, bone2Name=hand_l, swing1LimitDegrees=45, swing2LimitDegrees=45, twistLimitDegrees=45}
 {bone1Name=lowerarm_twist_01_l, bone2Name=lowerarm_l, swing1LimitDegrees=45, swing2LimitDegrees=45, twistLimitDegrees=45}
 {bone1Name=upperarm_twist_01_l, bone2Name=upperarm_l, swing1LimitDegrees=45, swing2LimitDegrees=45, twistLimitDegrees=45}
 {bone1Name=clavicle_r, bone2Name=spine_05, swing1LimitDegrees=45, swing2LimitDegrees=45, twistLimitDegrees=45}
 {bone1Name=upperarm_r, bone2Name=clavicle_r, swing1LimitDegrees=45, swing2LimitDegrees=45, twistLimitDegrees=45}
 {bone1Name=lowerarm_r, bone2Name=upperarm_r, swing1LimitDegrees=45, swing2LimitDegrees=45, twistLimitDegrees=45}
 {bone1Name=hand_r, bone2Name=lowerarm_r, swing1LimitDegrees=45, swing2LimitDegrees=45, twistLimitDegrees=45}
 {bone1Name=pinky_metacarpal_r, bone2Name=hand_r, swing1LimitDegrees=45, swing2LimitDegrees=45, twistLimitDegrees=45}
 {bone1Name=lowerarm_twist_01_r, bone2Name=lowerarm_r, swing1LimitDegrees=45, swing2LimitDegrees=45, twistLimitDegrees=45}
 {bone1Name=upperarm_twist_01_r, bone2Name=upperarm_r, swing1LimitDegrees=45, swing2LimitDegrees=45, twistLimitDegrees=45}
 {bone1Name=neck_01, bone2Name=spine_05, swing1LimitDegrees=45, swing2LimitDegrees=45, twistLimitDegrees=45}
 {bone1Name=neck_02, bone2Name=neck_01, swing1LimitDegrees=45, swing2LimitDegrees=45, twistLimitDegrees=45}
 {bone1Name=head, bone2Name=neck_02, swing1LimitDegrees=45, swing2LimitDegrees=45, twistLimitDegrees=45}
 {bone1Name=thigh_l, bone2Name=pelvis, swing1LimitDegrees=45, swing2LimitDegrees=45, twistLimitDegrees=45}
 {bone1Name=calf_l, bone2Name=thigh_l, swing1LimitDegrees=45, swing2LimitDegrees=45, twistLimitDegrees=45}
 {bone1Name=calf_twist_02_l, bone2Name=calf_l, swing1LimitDegrees=45, swing2LimitDegrees=45, twistLimitDegrees=45}
 {bone1Name=foot_l, bone2Name=calf_l, swing1LimitDegrees=45, swing2LimitDegrees=45, twistLimitDegrees=45}
 {bone1Name=ball_l, bone2Name=foot_l, swing1LimitDegrees=45, swing2LimitDegrees=45, twistLimitDegrees=45}
 {bone1Name=thigh_twist_01_l, bone2Name=thigh_l, swing1LimitDegrees=45, swing2LimitDegrees=45, twistLimitDegrees=45}
 {bone1Name=thigh_r, bone2Name=pelvis, swing1LimitDegrees=45, swing2LimitDegrees=45, twistLimitDegrees=45}
 {bone1Name=calf_r, bone2Name=thigh_r, swing1LimitDegrees=45, swing2LimitDegrees=45, twistLimitDegrees=45}
 {bone1Name=calf_twist_02_r, bone2Name=calf_r, swing1LimitDegrees=45, swing2LimitDegrees=45, twistLimitDegrees=45}
 {bone1Name=foot_r, bone2Name=calf_r, swing1LimitDegrees=45, swing2LimitDegrees=45, twistLimitDegrees=45}
 {bone1Name=ball_r, bone2Name=foot_r, swing1LimitDegrees=45, swing2LimitDegrees=45, twistLimitDegrees=45}
 {bone1Name=thigh_twist_01_r, bone2Name=thigh_r, swing1LimitDegrees=45, swing2LimitDegrees=45, twistLimitDegrees=45}
]
