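# Physics asset for the medium A-pose male player skeleton, giving it per-bone collision bodies joined by joints.
# Single capsules cover root, pelvis, spine, neck, head and limbs, including twist bones, with only the pinky metacarpal beyond each hand.
# Constraints follow the bone hierarchy, all sharing identical swing and twist limit angles rather than per-joint tuning.
# Root, spine_01 and neck_02 carry identical oversized, uncentered capsules, unlike the fitted shapes elsewhere.
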PhysicsAsset /Game/Characters/Player/Male/Male_Avg_Base/MPR/Player_Medium_Apose_PhysicsAsset
bodies {
 ball_l [{shapeType=Capsule, length=7.2692, radius=6.7752, center=<2.1974 -4.1844 -0.2333>, rotation=<rot 5.7282 -90 25.5632>}]
 ball_r [{shapeType=Capsule, length=7.0722, radius=6.8063, center=<-2.2405 4.2072 0.0877>, rotation=<rot 4.7116 -90 26.4915>}]
 calf_l [{shapeType=Capsule, length=20.2566, radius=5.6591, center=<-19.929 -1.2437 -0.677>, rotation=<rot 0 87.1756 -89.2468>}]
 calf_r [{shapeType=Capsule, length=20.5062, radius=5.6556, center=<19.6939 1.2895 0.6165>, rotation=<rot 0 87.3103 -89.451>}]
 calf_twist_02_l [{shapeType=Capsule, length=10.0855, radius=6.1745, center=<0.3111 -2.6995 -0.4217>, rotation=<rot 0 83.4873 -89.1066>}]
 calf_twist_02_r [{shapeType=Capsule, length=10.0855, radius=6.1747, center=<-0.3114 2.6995 0.4215>, rotation=<rot 0 83.4868 -89.1067>}]
 clavicle_l [{shapeType=Capsule, length=8.2336, radius=12.0529, center=<8.4756 0.4734 0.2323>, rotation=<rot 0 90 -90>}]
 clavicle_r [{shapeType=Capsule, length=8.236, radius=12.0545, center=<-8.4776 -0.4719 -0.252>, rotation=<rot 0 -89.9999 -90.0001>}]
 foot_l [{shapeType=Capsule, length=9.8157, radius=8.2926, center=<-9.042 3.0383 0.0686>, rotation=<rot 0 -134.5811 -90.2824>}]
 foot_r [{shapeType=Capsule, length=9.8163, radius=8.2929, center=<9.0412 -3.0381 -0.0686>, rotation=<rot 0 45.4188 -89.7176>}]
 hand_l [{shapeType=Capsule, length=4.356, radius=2.8465, center=<1.9197 -1.2878 0.0389>, rotation=<rot 10.1245 -90 2.2388>}]
 hand_r [{shapeType=Capsule, length=4.3574, radius=2.8462, center=<-1.9363 1.2449 -0.0357>, rotation=<rot 12.9068 -90 2.9398>}]
 head [{shapeType=Capsule, length=10.9653, radius=9.1435, center=<3.0884 2.8733 -0.0065>, rotation=<rot 0 54.1673 -89.7633>}]
 lowerarm_l [{shapeType=Capsule, length=8.8381, radius=5.72, center=<6.8843 0.0165 -0.1622>, rotation=<rot 0 88.5143 -89.0606>}]
 lowerarm_r [{shapeType=Capsule, length=9.0106, radius=5.715, center=<-6.7733 0.667 0.1786>, rotation=<rot 0 82.8235 -89.7166>}]
 lowerarm_twist_01_l [{shapeType=Capsule, length=4.0948, radius=4.0114, center=<7.9083 -0.0723 0.323>, rotation=<rot 0 -166.9631 -86.2806>}]
 lowerarm_twist_01_r [{shapeType=Capsule, length=5.0712, radius=3.8494, center=<-6.4789 -0.5477 0.1797>, rotation=<rot 0 -103.7811 -79.4718>}]
 neck_01 [{shapeType=Capsule, length=7.7273, radius=6.9515, center=<2.3263 1.2568 -0.1495>, rotation=<rot 4.4352 -90 -1.008>}]
 neck_02 [{shapeType=Capsule, length=15.15, radius=15.15, rotation=<rot 0 90 -90>}]
 pelvis [{shapeType=Capsule, length=14.8701, radius=14.4099, center=<3.6918 -0.7449 0.0012>, rotation=<rot -0.0182 -90 0.0002>}]
 pinky_metacarpal_l [{shapeType=Capsule, length=3.9429, radius=2.7243, center=<3.5288 -0.3 -0.943>, rotation=<rot 0 90 -90>}]
 pinky_metacarpal_r [{shapeType=Capsule, length=4.4577, radius=2.73, center=<-3.0058 0.3011 0.9391>, rotation=<rot 0 -90.001 -89.9998>}]
 root [{shapeType=Capsule, length=15.15, radius=15.15, rotation=<rot -0.4351 -90 0>}]
 spine_01 [{shapeType=Capsule, length=15.15, radius=15.15, rotation=<rot 0 90 -90>}]
 spine_02 [{shapeType=Capsule, length=3.1359, radius=12.5423, center=<5.1969 3.4552 0.0001>, rotation=<rot 0 90 -90>}]
 spine_03 [{shapeType=Capsule, length=13.7154, radius=15.9901, center=<11.0115 -1.5529 -0.0009>, rotation=<rot 0 90 -90>}]
 spine_04 [{shapeType=Capsule, length=5.8998, radius=7.5774, center=<4.9773 10.8151 -0.0003>, rotation=<rot 0 90 -90>}]
 spine_05 [{shapeType=Capsule, length=8.6178, radius=5.5045, center=<7.3225 -0.025 0.0003>, rotation=<rot 0 -152.4105 -89.9991>}]
 thigh_l [{shapeType=Capsule, length=7.957, radius=7.286, center=<-29.3813 0.6223 0.3069>, rotation=<rot 0 52.117 -82.2825>}]
 thigh_r [{shapeType=Capsule, length=7.8942, radius=7.255, center=<29.3607 -0.5725 -0.3581>, rotation=<rot 0 48.7988 -85.8191>}]
 thigh_twist_01_l [{shapeType=Capsule, length=13.1126, radius=9.0744, center=<7.3376 1.0497 0.4059>, rotation=<rot 0 100.8516 -86.3904>}]
 thigh_twist_01_r [{shapeType=Capsule, length=13.1254, radius=9.0697, center=<-7.399 -1.0571 -0.4106>, rotation=<rot 0 100.8491 -86.3809>}]
 upperarm_l [{shapeType=Capsule, length=9.013, radius=8.4403, center=<5.4944 0.6116 -0.9099>, rotation=<rot 0 8.2349 -86.3639>}]
 upperarm_r [{shapeType=Capsule, length=8.9674, radius=8.4595, center=<-5.5166 -0.6041 0.9268>, rotation=<rot 0 9.025 -86.6766>}]
 upperarm_twist_01_l [{shapeType=Capsule, length=6.4307, radius=5.2361, center=<6.2281 -0.5423 -0.8153>, rotation=<rot 0 -4.151 -84.0147>}]
 upperarm_twist_01_r [{shapeType=Capsule, length=6.4367, radius=5.2248, center=<-6.2262 0.5438 0.8096>, rotation=<rot 0 -4.2364 -83.9479>}]
}
constraints [
 {bone1Name=pelvis, bone2Name=root, swing1LimitDegrees=45, swing2LimitDegrees=45, twistLimitDegrees=45}
 {bone1Name=spine_01, bone2Name=pelvis, swing1LimitDegrees=45, swing2LimitDegrees=45, twistLimitDegrees=45}
 {bone1Name=spine_02, bone2Name=spine_01, swing1LimitDegrees=45, swing2LimitDegrees=45, twistLimitDegrees=45}
 {bone1Name=spine_03, bone2Name=spine_02, swing1LimitDegrees=45, swing2LimitDegrees=45, twistLimitDegrees=45}
 {bone1Name=spine_04, bone2Name=spine_03, swing1LimitDegrees=45, swing2LimitDegrees=45, twistLimitDegrees=45}
 {bone1Name=spine_05, bone2Name=spine_04, swing1LimitDegrees=45, swing2LimitDegrees=45, twistLimitDegrees=45}
 {bone1Name=clavicle_l, bone2Name=spine_05, swing1LimitDegrees=45, swing2LimitDegrees=45, twistLimitDegrees=45}
 {bone1Name=upperarm_l, bone2Name=clavicle_l, swing1LimitDegrees=45, swing2LimitDegrees=45, twistLimitDegrees=45}
 {bone1Name=lowerarm_l, bone2Name=upperarm_l, swing1LimitDegrees=45, swing2LimitDegrees=45, twistLimitDegrees=45}
 {bone1Name=hand_l, bone2Name=lowerarm_l, swing1LimitDegrees=45, swing2LimitDegrees=45, twistLimitDegrees=45}
 {bone1Name=pinky_metacarpal_l, bone2Name=hand_l, swing1LimitDegrees=45, swing2LimitDegrees=45, twistLimitDegrees=45}
 {bone1Name=lowerarm_twist_01_l, bone2Name=lowerarm_l, swing1LimitDegrees=45, swing2LimitDegrees=45, twistLimitDegrees=45}
 {bone1Name=upperarm_twist_01_l, bone2Name=upperarm_l, swing1LimitDegrees=45, swing2LimitDegrees=45, twistLimitDegrees=45}
 {bone1Name=clavicle_r, bone2Name=spine_05, swing1LimitDegrees=45, swing2LimitDegrees=45, twistLimitDegrees=45}
 {bone1Name=upperarm_r, bone2Name=clavicle_r, swing1LimitDegrees=45, swing2LimitDegrees=45, twistLimitDegrees=45}
 {bone1Name=lowerarm_r, bone2Name=upperarm_r, swing1LimitDegrees=45, swing2LimitDegrees=45, twistLimitDegrees=45}
 {bone1Name=hand_r, bone2Name=lowerarm_r, swing1LimitDegrees=45, swing2LimitDegrees=45, twistLimitDegrees=45}
 {bone1Name=pinky_metacarpal_r, bone2Name=hand_r, swing1LimitDegrees=45, swing2LimitDegrees=45, twistLimitDegrees=45}
 {bone1Name=lowerarm_twist_01_r, bone2Name=lowerarm_r, swing1LimitDegrees=45, swing2LimitDegrees=45, twistLimitDegrees=45}
 {bone1Name=upperarm_twist_01_r, bone2Name=upperarm_r, swing1LimitDegrees=45, swing2LimitDegrees=45, twistLimitDegrees=45}
 {bone1Name=neck_01, bone2Name=spine_05, swing1LimitDegrees=45, swing2LimitDegrees=45, twistLimitDegrees=45}
 {bone1Name=neck_02, bone2Name=neck_01, swing1LimitDegrees=45, swing2LimitDegrees=45, twistLimitDegrees=45}
 {bone1Name=head, bone2Name=neck_02, swing1LimitDegrees=45, swing2LimitDegrees=45, twistLimitDegrees=45}
 {bone1Name=thigh_l, bone2Name=pelvis, swing1LimitDegrees=45, swing2LimitDegrees=45, twistLimitDegrees=45}
 {bone1Name=calf_l, bone2Name=thigh_l, swing1LimitDegrees=45, swing2LimitDegrees=45, twistLimitDegrees=45}
 {bone1Name=calf_twist_02_l, bone2Name=calf_l, swing1LimitDegrees=45, swing2LimitDegrees=45, twistLimitDegrees=45}
 {bone1Name=foot_l, bone2Name=calf_l, swing1LimitDegrees=45, swing2LimitDegrees=45, twistLimitDegrees=45}
 {bone1Name=ball_l, bone2Name=foot_l, swing1LimitDegrees=45, swing2LimitDegrees=45, twistLimitDegrees=45}
 {bone1Name=thigh_twist_01_l, bone2Name=thigh_l, swing1LimitDegrees=45, swing2LimitDegrees=45, twistLimitDegrees=45}
 {bone1Name=thigh_r, bone2Name=pelvis, swing1LimitDegrees=45, swing2LimitDegrees=45, twistLimitDegrees=45}
 {bone1Name=calf_r, bone2Name=thigh_r, swing1LimitDegrees=45, swing2LimitDegrees=45, twistLimitDegrees=45}
 {bone1Name=calf_twist_02_r, bone2Name=calf_r, swing1LimitDegrees=45, swing2LimitDegrees=45, twistLimitDegrees=45}
 {bone1Name=foot_r, bone2Name=calf_r, swing1LimitDegrees=45, swing2LimitDegrees=45, twistLimitDegrees=45}
 {bone1Name=ball_r, bone2Name=foot_r, swing1LimitDegrees=45, swing2LimitDegrees=45, twistLimitDegrees=45}
 {bone1Name=thigh_twist_01_r, bone2Name=thigh_r, swing1LimitDegrees=45, swing2LimitDegrees=45, twistLimitDegrees=45}
]
