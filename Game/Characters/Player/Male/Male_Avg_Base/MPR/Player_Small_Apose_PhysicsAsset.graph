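# Physics asset for the small A-pose male player character, giving its skeleton collision bodies and joints.
# One capsule per bone covers root, pelvis, spine, neck, head, limbs, selected twist bones and pinky metacarpals, with no other finger bodies.
# Constraints link these bodies along the bone hierarchy, all with identical swing and twist limit values.
# Root, spine_01 and neck_02 carry identical large capsules, unlike the individually sized bodies elsewhere.
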PhysicsAsset /Game/Characters/Player/Male/Male_Avg_Base/MPR/Player_Small_Apose_PhysicsAsset
bodies {
 ball_l [{shapeType=Capsule, length=7.343, radius=6.6536, center=<0.9441 -4.23 -0.9434>, rotation=<rot 9.6152 -90 35.6325>}]
 ball_r [{shapeType=Capsule, length=7.1901, radius=6.7162, center=<-0.9869 4.198 0.9211>, rotation=<rot 6.6836 -90 35.0829>}]
 calf_l [{shapeType=Capsule, length=20.7149, radius=4.8124, center=<-19.1228 -1.4059 -0.2145>, rotation=<rot 0 -91.0253 -89.8589>}]
 calf_r [{shapeType=Capsule, length=21.2392, radius=4.8174, center=<18.5936 1.3745 0.1518>, rotation=<rot 0 -91.1868 -89.6255>}]
 calf_twist_02_l [{shapeType=Capsule, length=9.7873, radius=5.2569, center=<0.9754 -3.1237 -0.2743>, rotation=<rot 0 -93.3562 -88.3623>}]
 calf_twist_02_r [{shapeType=Capsule, length=9.7873, radius=5.2569, center=<-0.9755 3.1237 0.2743>, rotation=<rot 0 -93.3562 -88.3623>}]
 clavicle_l [{shapeType=Capsule, length=6.6405, radius=8.7229, center=<7.0837 -0.0944 -1.0515>, rotation=<rot 0 90 -90>}]
 clavicle_r [{shapeType=Capsule, length=6.6405, radius=8.7229, center=<-7.0837 0.0944 1.0516>, rotation=<rot 0 -89.9999 -90.0001>}]
 foot_l [{shapeType=Capsule, length=9.8225, radius=8.2805, center=<-9.0403 2.5671 -0.6964>, rotation=<rot 0 -134.5811 -90.2824>}]
 foot_r [{shapeType=Capsule, length=9.7928, radius=8.2805, center=<9.0192 -2.5465 0.6963>, rotation=<rot 0 45.4188 -89.7176>}]
 hand_l [{shapeType=Capsule, length=4.357, radius=2.8329, center=<1.8807 -1.0196 -0.247>, rotation=<rot 10.9868 -90 1.2773>}]
 hand_r [{shapeType=Capsule, length=4.3487, radius=2.8321, center=<-1.9043 0.973 0.2513>, rotation=<rot 13.7349 -90 1.7261>}]
 head [{shapeType=Capsule, length=11.1514, radius=8.0867, center=<2.8115 1.9899 0.0002>, rotation=<rot 0 -118.1957 -89.989>}]
 lowerarm_l [{shapeType=Capsule, length=9.9758, radius=3.9644, center=<5.6872 0.3999 -0.1761>, rotation=<rot 0 81.1757 -88.8388>}]
 lowerarm_r [{shapeType=Capsule, length=10.0824, radius=4.0511, center=<-5.6064 -0.1882 0.1042>, rotation=<rot 0 77.5633 -88.0799>}]
 lowerarm_twist_01_l [{shapeType=Capsule, length=4.6617, radius=3.7387, center=<7.7539 0.6969 -0.3645>, rotation=<rot 0 -103.5257 -86.0341>}]
 lowerarm_twist_01_r [{shapeType=Capsule, length=5.9773, radius=3.4131, center=<-6.1778 -0.9254 0.403>, rotation=<rot 0 -98.2271 -87.869>}]
 neck_01 [{shapeType=Capsule, length=7.5808, radius=6.2137, center=<0.8438 1.6007 0.0498>, rotation=<rot 0 -149.3474 -74.7108>}]
 neck_02 [{shapeType=Capsule, length=15.15, radius=15.15, rotation=<rot 0 90 -90>}]
 pelvis [{shapeType=Capsule, length=16.678, radius=12.4856, center=<5.5979 -1.0486 -0.1567>, rotation=<rot 0 -91.9338 -88.4073>}]
 pinky_metacarpal_l [{shapeType=Capsule, length=3.9615, radius=2.7002, center=<3.4225 -0.0542 -1.1505>, rotation=<rot 0 90 -90>}]
 pinky_metacarpal_r [{shapeType=Capsule, length=4.4645, radius=2.7002, center=<-2.9243 0.0543 1.1505>, rotation=<rot 0 -90.001 -89.9998>}]
 root [{shapeType=Capsule, length=15.15, radius=15.15, rotation=<rot -0.4351 -90 0>}]
 spine_01 [{shapeType=Capsule, length=15.15, radius=15.15, rotation=<rot 0 90 -90>}]
 spine_02 [{shapeType=Capsule, length=2.8379, radius=11.6051, center=<8.9783 2.2567 0>, rotation=<rot 0 90 -90>}]
 spine_03 [{shapeType=Capsule, length=12.5813, radius=13.2374, center=<13.4921 -2.1569 0>, rotation=<rot 0 90 -90>}]
 spine_04 [{shapeType=Capsule, length=4.0691, radius=6.2935, center=<6.262 7.0605 -0.0001>, rotation=<rot 0 90 -90>}]
 spine_05 [{shapeType=Capsule, length=7.4055, radius=4.6079, center=<6.8673 -0.4031 0.0037>, rotation=<rot 0 35.5101 -89.9708>}]
 thigh_l [{shapeType=Capsule, length=8.4261, radius=5.1893, center=<-27.5775 0.4809 -0.4071>, rotation=<rot 0 81.6347 -86.0095>}]
 thigh_r [{shapeType=Capsule, length=8.5076, radius=5.379, center=<27.5768 -0.5067 0.3587>, rotation=<rot 0 79.2613 -87.1906>}]
 thigh_twist_01_l [{shapeType=Capsule, length=12.222, radius=7.299, center=<9.9712 1.1135 1.4136>, rotation=<rot 0 105.8369 -83.226>}]
 thigh_twist_01_r [{shapeType=Capsule, length=12.222, radius=7.299, center=<-10.0208 -1.1143 -1.4158>, rotation=<rot 0 105.8369 -83.226>}]
 upperarm_l [{shapeType=Capsule, length=8.0573, radius=5.8085, center=<5.2552 -0.4425 -0.0821>, rotation=<rot 0 -96.4029 -64.0663>}]
 upperarm_r [{shapeType=Capsule, length=8.0573, radius=5.8085, center=<-5.2553 0.4425 0.0821>, rotation=<rot 0 -96.4027 -64.0662>}]
 upperarm_twist_01_l [{shapeType=Capsule, length=4.222, radius=3.7607, center=<3.9913 0.1933 -0.492>, rotation=<rot 0 -82.6747 -86.8409>}]
 upperarm_twist_01_r [{shapeType=Capsule, length=4.222, radius=3.7607, center=<-3.9914 -0.1933 0.4921>, rotation=<rot 0 -82.6747 -86.841>}]
}
constraints [
 {bone1Name=pelvis, bone2Name=root, swing1LimitDegrees=45, swing2LimitDegrees=45, twistLimitDegrees=45}
 {bone1Name=spine_01, bone2Name=pelvis, swing1LimitDegrees=45, swing2LimitDegrees=45, twistLimitDegrees=45}
 {bone1Name=spine_02, bone2Name=spine_01, swing1LimitDegrees=45, swing2LimitDegrees=45, twistLimitDegrees=45}
 {bone1Name=spine_03, bone2Name=spine_02, swing1LimitDegrees=45, swing2LimitDegrees=45, twistLimitDegrees=45}
 {bone1Name=spine_04, bone2Name=spine_03, swing1LimitDegrees=45, swing2LimitDegrees=45, twistLimitDegrees=45}
 {bone1Name=spine_05, bone2Name=spine_04, swing1LimitDegrees=45, swing2LimitDegrees=45, twistLimitDegrees=45}
 {bone1Name=clavicle_l, bone2Name=spine_05, swing1LimitDegrees=45, swing2LimitDegrees=45, twistLimitDegrees=45}
 {bone1Name=upperarm_l, bone2Name=clavicle_l, swing1LimitDegrees=45, swing2LimitDegrees=45, twistLimitDegrees=45}
 {bone1Name=lowerarm_l, bone2Name=upperarm_l, swing1LimitDegrees=45, swing2LimitDegrees=45, twistLimitDegrees=45}
 {bone1Name=hand_l, bone2Name=lowerarm_l, swing1LimitDegrees=45, swing2LimitDegrees=45, twistLimitDegrees=45}
 {bone1Name=pinky_metacarpal_l, bone2Name=hand_l, swing1LimitDegrees=45, swing2LimitDegrees=45, twistLimitDegrees=45}
 {bone1Name=lowerarm_twist_01_l, bone2Name=lowerarm_l, swing1LimitDegrees=45, swing2LimitDegrees=45, twistLimitDegrees=45}
 {bone1Name=upperarm_twist_01_l, bone2Name=upperarm_l, swing1LimitDegrees=45, swing2LimitDegrees=45, twistLimitDegrees=45}
 {bone1Name=clavicle_r, bone2Name=spine_05, swing1LimitDegrees=45, swing2LimitDegrees=45, twistLimitDegrees=45}
 {bone1Name=upperarm_r, bone2Name=clavicle_r, swing1LimitDegrees=45, swing2LimitDegrees=45, twistLimitDegrees=45}
 {bone1Name=lowerarm_r, bone2Name=upperarm_r, swing1LimitDegrees=45, swing2LimitDegrees=45, twistLimitDegrees=45}
 {bone1Name=hand_r, bone2Name=lowerarm_r, swing1LimitDegrees=45, swing2LimitDegrees=45, twistLimitDegrees=45}
 {bone1Name=pinky_metacarpal_r, bone2Name=hand_r, swing1LimitDegrees=45, swing2LimitDegrees=45, twistLimitDegrees=45}
 {bone1Name=lowerarm_twist_01_r, bone2Name=lowerarm_r, swing1LimitDegrees=45, swing2LimitDegrees=45, twistLimitDegrees=45}
 {bone1Name=upperarm_twist_01_r, bone2Name=upperarm_r, swing1LimitDegrees=45, swing2LimitDegrees=45, twistLimitDegrees=45}
 {bone1Name=neck_01, bone2Name=spine_05, swing1LimitDegrees=45, swing2LimitDegrees=45, twistLimitDegrees=45}
 {bone1Name=neck_02, bone2Name=neck_01, swing1LimitDegrees=45, swing2LimitDegrees=45, twistLimitDegrees=45}
 {bone1Name=head, bone2Name=neck_02, swing1LimitDegrees=45, swing2LimitDegrees=45, twistLimitDegrees=45}
 {bone1Name=thigh_l, bone2Name=pelvis, swing1LimitDegrees=45, swing2LimitDegrees=45, twistLimitDegrees=45}
 {bone1Name=calf_l, bone2Name=thigh_l, swing1LimitDegrees=45, swing2LimitDegrees=45, twistLimitDegrees=45}
 {bone1Name=calf_twist_02_l, bone2Name=calf_l, swing1LimitDegrees=45, swing2LimitDegrees=45, twistLimitDegrees=45}
 {bone1Name=foot_l, bone2Name=calf_l, swing1LimitDegrees=45, swing2LimitDegrees=45, twistLimitDegrees=45}
 {bone1Name=ball_l, bone2Name=foot_l, swing1LimitDegrees=45, swing2LimitDegrees=45, twistLimitDegrees=45}
 {bone1Name=thigh_twist_01_l, bone2Name=thigh_l, swing1LimitDegrees=45, swing2LimitDegrees=45, twistLimitDegrees=45}
 {bone1Name=thigh_r, bone2Name=pelvis, swing1LimitDegrees=45, swing2LimitDegrees=45, twistLimitDegrees=45}
 {bone1Name=calf_r, bone2Name=thigh_r, swing1LimitDegrees=45, swing2LimitDegrees=45, twistLimitDegrees=45}
 {bone1Name=calf_twist_02_r, bone2Name=calf_r, swing1LimitDegrees=45, swing2LimitDegrees=45, twistLimitDegrees=45}
 {bone1Name=foot_r, bone2Name=calf_r, swing1LimitDegrees=45, swing2LimitDegrees=45, twistLimitDegrees=45}
 {bone1Name=ball_r, bone2Name=foot_r, swing1LimitDegrees=45, swing2LimitDegrees=45, twistLimitDegrees=45}
 {bone1Name=thigh_twist_01_r, bone2Name=thigh_r, swing1LimitDegrees=45, swing2LimitDegrees=45, twistLimitDegrees=45}
]
